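# Niagara deploy effect for the male Clash glider, built from goop-like ribbons on user-supplied skeletal meshes.
# BoneFlow spawns seven short ribbon groups that FN_BoneFlow animates around bone splines, with curves shaping radius, height and width.
# Two GoopRibbonsBurst emitters, one delayed, throw short-lived skeletal-sampled ribbons with decaying spawn rates and curl noise.
# Front-end modifiers retune some values for menus; both droplet sprite emitters are disabled.
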NiagaraSystem /BRCosmetics/Gadgets/Assets/VinderTech_GliderChute/Glider_Clash_Male/FX/NS_Glider_Clash_Deploy
dependencies {
 usedDataInterfaces [&NiagaraDataInterfaceCurve, &NiagaraDataInterfaceArrayFloat3]
 usedDynamicInputs []
 usedModules [&SystemState]
}
emitters {
 BoneFlow {
  emitterSpawnScript [
   /SetVariables(Emitter.RandomSeed:NiagaraInt32=/Niagara/DynamicInputs/Random/RandomInt(Randomness Mode:ENiagaraRandomnessMode=::NewEnumerator0, Random Seed:NiagaraInt32=0, Fixed Random Seed:NiagaraBool=false)) {isSetParameters=true}
   /SetVariables(Emitter.ParticlesPerGroup=5) {isSetParameters=true}
  ]
  emitterUpdateScript [
   /SetVariables(Emitter.NumberOfGroups=7) {isSetParameters=true}
   &EmitterState(Life Cycle Mode=::NewEnumerator1, Loop Behavior=::NewEnumerator1, Loop Duration=1, UseLoopDelay=false)
   &SpawnBurst_Instantaneous(
    Spawn Count=&Multiply_Int(A=$Emitter.ParticlesPerGroup, B=$Emitter.NumberOfGroups)
    Use Spawn Probability=false
   )
  ]
  particleSpawnScript [
   /SetVariables(Particles.LoopCount:NiagaraInt32=$Emitter.LoopCount) {isSetParameters=true}
   &FN_SetSpawnGroup(
    Inclusivity=true
    Normalize Spawn Group=true
    Particles Per Group:NiagaraInt32=$Emitter.ParticlesPerGroup
    Pattern=::NewEnumerator0
    Set Ribbon ID and Ribbon Link Order=true
   )
   &InitializeRibbon(
    Lifetime=2
    Module.Write Color=true
    Module.Write Lifetime=true
    Module.Write Mass=true
    Module.Write Material Random=true
    Module.Write Position=false
    Module.Write RibbonTwist=false
    Module.Write RibbonWidth=true
    Ribbon Width=&RandomRangeFloat(Minimum=4, Maximum=7)
   )
   &FN_FlipRibbonV()
  ]
  particleUpdateScript [
   &ParticleState()
   /Game/Effects/Niagara/Modules/BoneFlow/FN_BoneFlow_v4/FN_BoneFlow(
    Angle:NiagaraFloat=&Add_Float(A=$Particles.Age, B=&Lerp_Float(A=0, B=1, Alpha=$Particles.Normalized Spawn Group))
    BoneID:NiagaraInt32=&ModuloInt(A=$Particles.SpawnGroup, B=7)
    Height:NiagaraFloat=&Multiply_Float(
     A=$Particles.Normalized Spawn Group
     B=&FloatFromCurve(
      FloatCurve=DataInterface() {propertyValues {Curve {keys [
          {arriveTangent=2.2162, interpMode=RCIM_Cubic, leaveTangent=2.2162, tangentMode=RCTM_User, time=0, value=0}
          {arriveTangent=-0.017, interpMode=RCIM_Cubic, leaveTangent=-0.017, tangentMode=RCTM_User, time=0.8, value=1}
          {arriveTangent=-4.7911, interpMode=RCIM_Cubic, leaveTangent=-4.7911, tangentMode=RCTM_User, time=1, value=0.15}
         ]}}}
      Scale Curve=&RandomRangeFloat(Minimum=1, Maximum=1)
     )
    )
    Interpolate Mesh Sampling=false
    Radius:NiagaraFloat=&FloatFromCurve(
     FloatCurve=DataInterface() {propertyValues {Curve {keys [
         {arriveTangent=5.7764, interpMode=RCIM_Cubic, leaveTangent=5.7764, tangentMode=RCTM_User, time=0, value=0}
         {arriveTangent=-0.031, interpMode=RCIM_Cubic, leaveTangent=-0.031, tangentMode=RCTM_User, time=0.5, value=1}
         {arriveTangent=-0.3846, interpMode=RCIM_Cubic, leaveTangent=-0.3846, tangentMode=RCTM_User, time=1, value=0}
        ]}}}
     CurveIndex=$Particles.Normalized Spawn Group
     Scale Curve=&RandomRangeFloat(Minimum=90, Maximum=80)
    )
    SideVector1:Vector3f=<0 0 0>
    SideVector2:Vector3f=&RandomRangeVector(Minimum=<0 -1 0>, Maximum=<0 1 0>)
    SideVector3:Vector3f=<0 0 0>
    SkeletalMesh:NiagaraDataInterfaceSkeletalMesh=$User.Skeletal Mesh1
    Stabalize Twisting=true
    Write Position Attribute=true
   ) {module=FN_BoneFlow001}
   /SetVariables(Particles.RibbonFacing=$Output.FN_BoneFlow001.SplineBinormal) {isSetParameters=true}
   &ScaleRibbonWidth(Ribbon Width Scale=&FloatFromCurve(
     FloatCurve=DataInterface() {propertyValues {Curve {keys [
         {arriveTangent=7.7351, interpMode=RCIM_Cubic, leaveTangent=7.7351, tangentMode=RCTM_User, time=0, value=0}
         {interpMode=RCIM_Cubic, tangentMode=RCTM_User, time=0.5, value=1}
         {arriveTangent=-6.7655, interpMode=RCIM_Cubic, leaveTangent=-6.7655, tangentMode=RCTM_User, time=1, value=0}
        ]}}}
     CurveIndex=$Particles.Normalized Spawn Group
     Scale Curve=&Multiply_Float(
      A=2.35
      B=&FloatFromCurve(
       FloatCurve=DataInterface() {propertyValues {Curve {keys [
           {arriveTangent=2.8989, interpMode=RCIM_Cubic, leaveTangent=2.8989, tangentMode=RCTM_User, time=0, value=0}
           {interpMode=RCIM_Cubic, tangentMode=RCTM_User, time=0.5, value=1}
           {arriveTangent=-2.6878, interpMode=RCIM_Cubic, leaveTangent=-2.6878, tangentMode=RCTM_User, time=1, value=0}
          ]}}}
       CurveIndex=$Particles.Age
      )
     )
    ))
   &DynamicMaterialParameters(
    Index 0 Param 1=&FloatFromCurve(FloatCurve=DataInterface() {propertyValues {Curve {keys [<at 0 0 RCIM_Cubic>,<at 0.5 1>,<at 1 0 RCIM_Cubic>]}}})
    Module.Param0WriteEnabled=false
    Module.Param1WriteEnabled=false
    Module.Param2WriteEnabled=false
    Module.Param3WriteEnabled=false
   )
  ]
  rendererClasses [&NiagaraRibbonRendererProperties]
  renderers [{rendererClass=&NiagaraRibbonRendererProperties}]
 }
 DropsBurst {
  bEnabled=false
  emitterUpdateScript [
   &EmitterState(
    Life Cycle Mode=::NewEnumerator1
    Loop Behavior=::NewEnumerator1
    Loop Delay=&FN_FrontEndFloatModifier(Gameplay=0.3, Front End=0.4)
    Loop Duration=1
    UseLoopDelay=true
   )
   &SpawnBurst_Instantaneous(Spawn Count=18, Use Spawn Probability=false, UseLoopCountLimit=true)
   &SpawnRate(
    SpawnRate=&FloatFromCurve(
     FloatCurve=DataInterface() {propertyValues {Curve {keys [{interpMode=RCIM_Cubic, tangentMode=RCTM_User, time=0, value=1}, {interpMode=RCIM_Cubic, tangentMode=RCTM_User, time=1, value=0}]}}}
     CurveIndex=$System.Age
     Scale Curve=6
    )
    Use Spawn Probability=false
   )
  ]
  particleSpawnScript [
   &FN_SetSpawnGroup(Number of Spawn Groups=6)
   &InitializeParticle(
    Color Mode=::NewEnumerator0
    Lifetime Max=0.7
    Lifetime Min=0.5
    Lifetime Mode=::NewEnumerator1
    Sprite Size Max=<10 20>
    Sprite Size Min=<7 14>
    Sprite Size Mode=::NewEnumerator2
    UsePositionOffset=false
    Write Material Random=true
   )
   &SkeletalMeshLocation(Bone / Socket Index=$Particles.SpawnGroup, Bone Sampling Mode=::NewEnumerator4, MaskSpawnGroup=false, Skeletal Mesh=$User.Skeletal Mesh2)
   &AddVelocity(
    Coordinate Space=::World
    Velocity=&RotateVectorByQuaternion(
     Vector To Rotate=&SelectVectorFromArray(
      Direct Array Index=$Particles.SpawnGroup
      Vector Selection Array=DataInterface() {propertyValues {FloatData [<-100 0 0>, <-100 0 0>, <-100 0 0>, <100 0 0>, <100 0 0>, <100 0 0>]}}
     )
     Quaternion=$Particles.SkeletalMeshLocation.SampledRotation
    )
   ) {module=AddVelocity001}
   &AddVelocityFromPoint(Module.Use Falloff Distance=false, Velocity Strength=&RandomRangeFloat(Minimum=150, Maximum=300))
   &SubUVAnimation(End Frame=3, SubUV Animation Mode:ENiagara_SubUVLookupMode=::NewEnumerator2)
  ]
  particleUpdateScript [
   &ParticleState(Kill Particles When Lifetime Has Elapsed=false, Let Infinitely Lived Particles Die When Emitter Deactivates=true)
   &GravityForce(Gravity=&RandomRangeVector(Minimum=<0 0 -300>, Maximum=<0 0 -600>))
   &Drag(Drag=&RandomRangeFloat(Minimum=2, Maximum=3), Use Linear Drag=true, Use Rotational Drag=false)
   &InheritSourceMovement(Applied Position Scale=<0.5 0.5 1>, ApplyForce=false, ApplyPosition=true, ApplyPositionFalloff=false)
   &SolveForcesAndVelocity(Clamp Velocity=false, Limit Acceleration=false)
   &ScaleSpriteSize(Scale Factor=&Vector2DFromFloat(Value=&FloatFromCurve(FloatCurve=DataInterface() {propertyValues {Curve {keys [<at 0 1 RCIM_Cubic>, {arriveTangent=-1.5322, interpMode=RCIM_Cubic, leaveTangent=-1.5322, tangentMode=RCTM_User, time=1, value=0}]}}})))
  ]
  rendererClasses [&NiagaraSpriteRendererProperties]
  renderers [{rendererClass=&NiagaraSpriteRendererProperties}]
 }
 DropsBurst001 {
  bEnabled=false
  emitterUpdateScript [
   &EmitterState(
    Life Cycle Mode=::NewEnumerator1
    Loop Behavior=::NewEnumerator1
    Loop Delay=&FN_FrontEndFloatModifier(Gameplay=0, Front End=0.2)
    Loop Duration=1
    UseLoopDelay=true
   )
   &SpawnBurst_Instantaneous(Spawn Count=20, Use Spawn Probability=false, UseLoopCountLimit=true)
   &SpawnRate(
    SpawnRate=&FloatFromCurve(
     FloatCurve=DataInterface() {propertyValues {Curve {keys [{interpMode=RCIM_Cubic, tangentMode=RCTM_User, time=0, value=1}, {interpMode=RCIM_Cubic, tangentMode=RCTM_User, time=1, value=0}]}}}
     CurveIndex=$System.Age
     Scale Curve=&FN_FrontEndFloatModifier(Gameplay=10, Front End=0)
    )
    Use Spawn Probability=false
   )
  ]
  particleSpawnScript [
   &InitializeParticle(
    Color Mode=::NewEnumerator0
    Lifetime Max=1
    Lifetime Min=0.5
    Lifetime Mode=::NewEnumerator1
    Position Mode=::NewEnumerator1
    Sprite Size Max=<15 30>
    Sprite Size Min=<10 15>
    Sprite Size Mode=::NewEnumerator2
    UsePositionOffset=false
    Write Material Random=true
   )
   &SkeletalMeshLocation(
    Bone Sampling Mode=::NewEnumerator4
    MaskSpawnGroup=false
    Orientation Sampling=::NewEnumerator1
    Position Sampling=::NewEnumerator2
    Skeletal Mesh=$User.SkeletalMesh
   )
   &AddVelocity(Coordinate Space=::World, Velocity=&RandomRangeVector(Minimum=<0 0 100>, Maximum=<0 0 200>)) {module=AddVelocity001}
   &AddVelocityFromPoint(
    Module.Use Falloff Distance=false
    Velocity Origin:Vector3f=<0 0 0>
    Velocity Strength=&FN_FrontEndFloatModifier(
     Gameplay=&RandomRangeFloat(Minimum=300, Maximum=600)
     Front End=&RandomRangeFloat(Minimum=300, Maximum=1000)
    )
   )
   &SubUVAnimation(End Frame=3, SubUV Animation Mode:ENiagara_SubUVLookupMode=::NewEnumerator2)
  ]
  particleUpdateScript [
   &ParticleState()
   &GravityForce(Gravity=&RandomRangeVector(Minimum=<0 0 -300>, Maximum=<0 0 -900>))
   &Drag(Drag=&RandomRangeFloat(Minimum=1, Maximum=2), Use Linear Drag=true, Use Rotational Drag=false)
   &SolveForcesAndVelocity(Clamp Velocity=false, Limit Acceleration=false)
   &ScaleSpriteSize(Scale Factor=&Vector2DFromFloat(Value=&FloatFromCurve(
      FloatCurve=DataInterface() {propertyValues {Curve {keys [<at 0 1 RCIM_Cubic>, {arriveTangent=-1.5322, interpMode=RCIM_Cubic, leaveTangent=-1.5322, tangentMode=RCTM_User, time=1, value=0}]}}}
      Scale Curve=&FN_FrontEndFloatModifier(Gameplay=1, Front End=0.8)
     )))
  ]
  rendererClasses [&NiagaraSpriteRendererProperties]
  renderers [{rendererClass=&NiagaraSpriteRendererProperties}]
 }
 GoopRibbonsBurst {
  emitterUpdateScript [
   &EmitterState(
    Life Cycle Mode=::NewEnumerator1
    Loop Behavior=::NewEnumerator1
    Loop Delay=&FN_FrontEndFloatModifier(Gameplay=0, Front End=0.15)
    Loop Duration=0.6
    UseLoopDelay=true
   )
   &FN_SpawnRibbon_v2(
    Number of Ribbons=11
    SpawnRate:NiagaraFloat=&FloatFromCurve(
     FloatCurve=DataInterface() {propertyValues {Curve {keys [{interpMode=RCIM_Cubic, tangentMode=RCTM_User, time=0, value=1}, {interpMode=RCIM_Cubic, tangentMode=RCTM_User, time=1, value=0}]}}}
     CurveIndex=$System.Age
     Scale Curve=33
    )
   )
  ]
  particleSpawnScript [
   /SetVariables(Particles.MaterialRandom:NiagaraFloat=&RandomRangeFloat(Minimum=0, Maximum=1)) {isSetParameters=true, module=SetVariables_#1}
   &FN_InitializeRibbon_v3(
    Enable Skeletal Position=true
    Lifetime:NiagaraFloat=&FN_FrontEndFloatModifier(Gameplay=0.15, Front End=0.1)
    Ribbon Width:NiagaraFloat=&FN_FrontEndFloatModifier(Gameplay=30, Front End=15)
    Sampled Position Coordinate Space:ENiagaraCoordinateSpace=::Local
    Sampled Rotation Coordinate Space:ENiagaraCoordinateSpace=::Local
    Sampled Velocity Coordinate Space:ENiagaraCoordinateSpace=::Local
   ) {module=FN_InitializeRibbon_v3001}
   /SetVariables(Particles.RandomizeLoopedEmitterAge:NiagaraFloat=&NormalizeFloat(
     Value=&Add_Float(A=&RandomRangeFloat(Minimum=0, Maximum=1), B=$Emitter.NormalizedLoopAge)
     Min=0
     Max=2
    )) {isSetParameters=true, module=SetVariables_#2}
   &DynamicMaterialParameters(
    Index 0 Param 1=&FloatFromCurve(
     FloatCurve=DataInterface() {propertyValues {Curve {keys [
         <at 0 1>
         {arriveTangent=-1.5667, leaveTangent=-1.5667, time=0.2, value=0.9}
         <at 0.3 -0.425>
         {arriveTangent=-3.441, leaveTangent=-3.441, time=0.4, value=-1.1}
        ]}}}
     CurveIndex=$Particles.RandomizeLoopedEmitterAge
    )
    Index 0 Param 2=$Particles.MaterialRandom
    Index 0 Param 3=5
    Module.Param0WriteEnabled=true
    Module.Param1WriteEnabled=true
    Module.Param2WriteEnabled=true
    Module.Param3WriteEnabled=true
   ) {enabled=false, module=DynamicMaterialParameters001}
   &AddVelocity(
    Coordinate Space=::Simulation
    Velocity=&RotateVectorByQuaternion(
     Vector To Rotate=&SelectVectorFromArray(
      Direct Array Index=$Particles.SpawnGroup
      Vector Selection Array=DataInterface() {propertyValues {FloatData [<-100 0 0>, <-100 0 0>, <-100 0 0>, <100 0 0>, <100 0 0>, <100 0 0>]}}
     )
     Quaternion=$Particles.SampledRotation
    )
   ) {module=AddVelocity001}
   &AddVelocityFromPoint(Module.Use Falloff Distance=false, Velocity Strength=&RandomRangeFloat(Minimum=-400, Maximum=400))
  ]
  particleUpdateScript [
   /Niagara/Modules/Update/Forces/CurlNoiseForce(Noise Frequency=50, Noise Strength=1768.481, Pan Noise=true, Pan Noise Field=<0 0 0.25>, Random Seed=2)
   &DynamicMaterialParameters(
    Index 0 Param 1=1
    Index 0 Param 2=$Particles.MaterialRandom
    Index 0 Param 3=5
    Module.Param0WriteEnabled=true
    Module.Param1WriteEnabled=true
    Module.Param2WriteEnabled=true
    Module.Param3WriteEnabled=true
   )
   &SolveForcesAndVelocity(Clamp Velocity=false, Limit Acceleration=false)
   &FN_UpdateRibbon(
    Detach Locked Particle=true
    Enable Skeletal Position=true
    Initial Color:LinearColor=$Particles.Initial.Color
    Initial RibbonWidth:NiagaraFloat=$Particles.Initial.RibbonWidth
    Locked Particle Detach Time=2
   )
   &ScaleRibbonWidth(Ribbon Width Scale=&FloatFromCurve(FloatCurve=DataInterface() {propertyValues {Curve {keys [
         {arriveTangent=2.9281, interpMode=RCIM_Cubic, leaveTangent=2.9281, tangentMode=RCTM_User, time=0, value=0}
         <at 0.3 1 RCIM_Cubic>
         {arriveTangent=-2.0105, interpMode=RCIM_Cubic, leaveTangent=-2.0105, tangentMode=RCTM_User, time=1, value=0}
        ]}}}))
  ]
  rendererClasses [&NiagaraRibbonRendererProperties]
  renderers [{rendererClass=&NiagaraRibbonRendererProperties}]
 }
 GoopRibbonsBurst001 {
  emitterUpdateScript [
   &EmitterState(Life Cycle Mode=::NewEnumerator1, Loop Behavior=::NewEnumerator1, Loop Delay=0.3, Loop Duration=2, UseLoopDelay=true)
   &FN_SpawnRibbon_v2(
    Number of Ribbons=6
    SpawnRate:NiagaraFloat=&FloatFromCurve(
     FloatCurve=DataInterface() {propertyValues {Curve {keys [{interpMode=RCIM_Cubic, tangentMode=RCTM_User, time=0, value=1}, {interpMode=RCIM_Cubic, tangentMode=RCTM_User, time=1, value=0}]}}}
     CurveIndex=$System.Age
     Scale Curve=30
    )
   )
  ]
  particleSpawnScript [
   /SetVariables(Particles.MaterialRandom:NiagaraFloat=&RandomRangeFloat(Minimum=0, Maximum=1)) {isSetParameters=true, module=SetVariables_#1}
   &FN_InitializeRibbon_v3(
    Enable Skeletal Position=true
    Lifetime=0.2
    Ribbon Width=30
    Sampled Position Coordinate Space:ENiagaraCoordinateSpace=::Local
    Sampled Rotation Coordinate Space:ENiagaraCoordinateSpace=::Local
    Sampled Velocity Coordinate Space:ENiagaraCoordinateSpace=::Local
   ) {module=FN_InitializeRibbon_v3001}
   /SetVariables(Particles.RandomizeLoopedEmitterAge:NiagaraFloat=&NormalizeFloat(
     Value=&Add_Float(A=&RandomRangeFloat(Minimum=0, Maximum=1), B=$Emitter.NormalizedLoopAge)
     Min=0
     Max=2
    )) {isSetParameters=true, module=SetVariables_#2}
   &DynamicMaterialParameters(
    Index 0 Param 1=&FloatFromCurve(
     FloatCurve=DataInterface() {propertyValues {Curve {keys [
         <at 0 1>
         {arriveTangent=-1.5667, leaveTangent=-1.5667, time=0.2, value=0.9}
         <at 0.3 -0.425>
         {arriveTangent=-3.441, leaveTangent=-3.441, time=0.4, value=-1.1}
        ]}}}
     CurveIndex=$Particles.RandomizeLoopedEmitterAge
    )
    Index 0 Param 2=$Particles.MaterialRandom
    Index 0 Param 3=5
    Module.Param0WriteEnabled=true
    Module.Param1WriteEnabled=true
    Module.Param2WriteEnabled=true
    Module.Param3WriteEnabled=true
   )
   &AddVelocity(
    Coordinate Space=::Simulation
    Velocity=&RotateVectorByQuaternion(
     Vector To Rotate=&SelectVectorFromArray(
      Direct Array Index=$Particles.SpawnGroup
      Vector Selection Array=DataInterface() {propertyValues {FloatData [<-300 0 0>, <-300 0 0>, <-300 0 0>, <300 0 0>, <300 0 0>, <300 0 0>]}}
     )
     Quaternion=$Particles.SampledRotation
    )
   ) {module=AddVelocity001}
   &AddVelocityFromPoint(Module.Use Falloff Distance=false, Velocity Strength=&RandomRangeFloat(Minimum=200, Maximum=400))
  ]
  particleUpdateScript [
   &CurlNoiseForce(
    Noise Frequency=1
    Noise Strength=1500
    Pan Noise=true
    Pan Noise Field=<0 0 0.25>
    Random Seed:NiagaraInt32=&RandomRangeInt(Minimum=0, Maximum=1)
   )
   &SolveForcesAndVelocity(Clamp Velocity=false, Limit Acceleration=false)
   &FN_UpdateRibbon(
    Detach Locked Particle=true
    Enable Skeletal Position=true
    Initial Color:LinearColor=$Particles.Initial.Color
    Initial RibbonWidth:NiagaraFloat=$Particles.Initial.RibbonWidth
    Locked Particle Detach Time=0.25
   )
   &ScaleRibbonWidth(Ribbon Width Scale=&FloatFromCurve(
     FloatCurve=DataInterface() {propertyValues {Curve {keys [
         {arriveTangent=2.9281, interpMode=RCIM_Cubic, leaveTangent=2.9281, tangentMode=RCTM_User, time=0, value=0}
         <at 0.3 1 RCIM_Cubic>
         {arriveTangent=-2.0105, interpMode=RCIM_Cubic, leaveTangent=-2.0105, tangentMode=RCTM_User, time=1, value=0}
        ]}}}
     Scale Curve=0.5
    ))
  ]
  rendererClasses [&NiagaraRibbonRendererProperties]
  renderers [{rendererClass=&NiagaraRibbonRendererProperties}]
 }
}
summary {
 systemName=NS_Glider_Clash_Deploy
 emitters [
  {bEnabled=false, emitterName=DropsBurst, rendererClasses [&NiagaraSpriteRendererProperties]}
  {bEnabled=false, emitterName=DropsBurst001, rendererClasses [&NiagaraSpriteRendererProperties]}
  {emitterName=GoopRibbonsBurst, rendererClasses [&NiagaraRibbonRendererProperties]}
  {emitterName=BoneFlow, rendererClasses [&NiagaraRibbonRendererProperties]}
  {emitterName=GoopRibbonsBurst001, rendererClasses [&NiagaraRibbonRendererProperties]}
 ]
 userVariables [
  User.bIsFrontEnd {type=NiagaraBool}
  User.bIsFrontEndPreview {type=NiagaraBool}
  User.RibbonIDOffset {type=NiagaraInt32}
  User.Skeletal Mesh1 {
   type=NiagaraDataInterfaceSkeletalMesh
   defaultValue {
    struct=NiagaraExt_VariableValue_DataInterface
    value {dataInterface=&self:NiagaraDataInterfaceSkeletalMesh_1}
   }
  }
  User.Skeletal Mesh2 {
   type=NiagaraDataInterfaceSkeletalMesh
   defaultValue {
    struct=NiagaraExt_VariableValue_DataInterface
    value {dataInterface=&self:NiagaraDataInterfaceSkeletalMesh_0}
   }
  }
  User.SkeletalMesh {
   type=NiagaraDataInterfaceSkeletalMesh
   defaultValue {
    struct=NiagaraExt_VariableValue_DataInterface
    value {dataInterface=&self:NiagaraDataInterfaceSkeletalMesh_2}
   }
  }
 ]
}
