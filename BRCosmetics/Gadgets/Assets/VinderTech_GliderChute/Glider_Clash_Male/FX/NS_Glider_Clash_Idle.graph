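# Idle Niagara effect for the male Clash glider chute, with SideWings as its only enabled emitter.
# SideWings spawns two short-lived anim-trail ribbons between skeletal bone or socket pairs, streamed along local Y.
# Front-end modifiers flip that direction and swap width and alpha; in gameplay, FN_Integrator's normalized value drives trail alpha.
# Disabled Drops, GoopRibbons and GoopRibbonsBurst emitters retain sprite-drip and curl-noise goop ribbon setups.
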NiagaraSystem /BRCosmetics/Gadgets/Assets/VinderTech_GliderChute/Glider_Clash_Male/FX/NS_Glider_Clash_Idle
dependencies {
 usedDataInterfaces [&NiagaraDataInterfaceArrayFloat3, &NiagaraDataInterfaceCurve, &NiagaraDataInterfaceArrayFloat]
 usedDynamicInputs []
 usedModules [&SystemState, &FN_Integrator]
}
emitters {
 Drops {
  bEnabled=false
  emitterUpdateScript [
   &EmitterState()
   /SetVariables(Emitter.SpawnNum=6) {isSetParameters=true}
   &SpawnRate(
    SpawnRate=&MakeCustomFloatFromBool(
     Bool=$User.bisFrontEnd
     True Float=10
     False Float=&MakeCustomFloatFromBool(
      Bool=$User.bisFullyDeployed
      True Float=&Lerp_Float(A=0, B=$Emitter.SpawnNum, Alpha=$System.FN_Integrator.Value)
      False Float=0
     )
    )
    Use Spawn Probability=false
   )
  ]
  particleSpawnScript [
   &FN_SetSpawnGroup(Number of Spawn Groups=6)
   &InitializeParticle(
    Color Mode=::NewEnumerator0
    Lifetime=0.5
    Sprite Size Max=<10 20>
    Sprite Size Min=<7 14>
    Sprite Size Mode=::NewEnumerator2
    UsePositionOffset=false
    Write Material Random=true
   )
   &SubUVAnimation(End Frame=3, SubUV Animation Mode:ENiagara_SubUVLookupMode=::NewEnumerator2)
   &SkeletalMeshLocation(
    Bone / Socket Index=$Particles.SpawnGroup
    MaskSpawnGroup=false
    Mesh Sampling Type=::NewEnumerator5
    Skeletal Mesh=$User.Skeletal Mesh Drops
    Socket Sampling Mode=::NewEnumerator5
   )
   &AddVelocity(Velocity=&FN_FrontEndVector3Modifier(
     Gameplay=&RandomRangeVector(Minimum=<0 150 0>, Maximum=<0 250 0>)
     Front End=&RandomRangeVector(Minimum=<0 -300 0>, Maximum=<0 -400 0>)
    )) {module=AddVelocity003}
  ]
  particleUpdateScript [
   &ParticleState(Kill Particles When Lifetime Has Elapsed=false, Let Infinitely Lived Particles Die When Emitter Deactivates=true)
   &Drag(
    Drag=&FN_FrontEndFloatModifier(Gameplay=3, Front End=0)
    Use Linear Drag=true
    Use Rotational Drag=false
   )
   &InheritVelocity(
    Module.Limit Speed=true
    Velocity Limit:NiagaraFloat=&FN_FrontEndFloatModifier(Gameplay=50, Front End=0)
    Velocity Scale=<-1 -1 0>
    Velocity Source=&TransformVector(Vector=$Engine.Owner.Velocity, Source Space=::World, Destination Space=::Local, Apply Scale=true)
   )
   &SolveForcesAndVelocity(Clamp Velocity=false, Limit Acceleration=false)
   &ScaleSpriteSize(Scale Factor=&Vector2DFromFloat(Value=&FloatFromCurve(FloatCurve=DataInterface() {propertyValues {Curve {keys [<at 0 1 RCIM_Cubic>, {arriveTangent=-1.5322, interpMode=RCIM_Cubic, leaveTangent=-1.5322, tangentMode=RCTM_User, time=1, value=0}]}}})))
  ]
  rendererClasses [&NiagaraSpriteRendererProperties]
  renderers [{rendererClass=&NiagaraSpriteRendererProperties}]
 }
 GoopRibbons {
  bEnabled=false
  emitterUpdateScript [
   &EmitterState(Delay First Loop Only=true, Life Cycle Mode=::NewEnumerator1, Loop Delay=0.5, Loop Duration=1, UseLoopDelay=true)
   &FN_SpawnRibbon_v2(
    Number of Ribbons=6
    SpawnRate:NiagaraFloat=&MakeCustomFloatFromBool(
     Bool=$User.bisFrontEnd
     True Float=10
     False Float=&MakeCustomFloatFromBool(Bool=$User.bisFullyDeployed, True Float=&Lerp_Float(A=0, B=10, Alpha=$System.FN_Integrator.Value), False Float=0)
    )
   )
  ]
  particleSpawnScript [
   /SetVariables(Particles.MaterialRandom:NiagaraFloat=&RandomRangeFloat(Minimum=0, Maximum=1)) {isSetParameters=true, module=SetVariables_#1}
   &FN_InitializeRibbon_v3(
    Enable Skeletal Position=true
    Ribbon Width=15
    Sampled Position Coordinate Space:ENiagaraCoordinateSpace=::Local
    Sampled Rotation Coordinate Space:ENiagaraCoordinateSpace=::Local
    Sampled Velocity Coordinate Space:ENiagaraCoordinateSpace=::Local
   ) {module=FN_InitializeRibbon_v3001}
   /SetVariables(Particles.RandomizeLoopedEmitterAge:NiagaraFloat=&NormalizeFloat(
     Value=&Add_Float(A=&RandomRangeFloat(Minimum=0, Maximum=1), B=$Emitter.NormalizedLoopAge)
     Min=0
     Max=2
    )) {isSetParameters=true, module=SetVariables_#2}
   /SetVariables(Particles.EndPart:NiagaraFloat=&SelectFloatFromArray(Direct Array Index=$Particles.SpawnGroup, Float Selection Array=DataInterface() {propertyValues {FloatData [1,1,1,1,1,1]}})) {isSetParameters=true}
   &AddVelocity(Velocity=&FN_FrontEndVector3Modifier(
     Gameplay=&RandomRangeVector(Minimum=<0 150 0>, Maximum=<0 250 0>)
     Front End=&RandomRangeVector(Minimum=<0 -150 0>, Maximum=<0 -250 0>)
    )) {module=AddVelocity003}
   &DynamicMaterialParameters(
    Index 0 Param 1=&FloatFromCurve(
     FloatCurve=DataInterface() {propertyValues {Curve {keys [
         <at 0 1>
         {arriveTangent=-1.5667, leaveTangent=-1.5667, time=0.3, value=1}
         <at 0.37 0>
         {arriveTangent=-3.441, leaveTangent=-3.441, time=0.5174, value=-1}
        ]}}}
     CurveIndex=$Particles.RandomizeLoopedEmitterAge
    )
    Index 0 Param 2=$Particles.MaterialRandom
    Index 0 Param 3=0.7
    Module.Param0WriteEnabled=true
    Module.Param1WriteEnabled=true
    Module.Param2WriteEnabled=false
    Module.Param3WriteEnabled=true
   )
   &KillParticlesInVolume(Box Size:Vector3f=<600 500 500>, Invert Volume=true, Kill Shape=::NewEnumerator1)
  ]
  particleUpdateScript [
   /Niagara/Modules/Update/Forces/CurlNoiseForce(Noise Frequency=1, Noise Strength=&Multiply_Float(A=300, B=$Particles.EndPart), Pan Noise=true)
   &SolveForcesAndVelocity(Clamp Velocity=false, Limit Acceleration=false)
   &FN_UpdateRibbon(
    Enable Skeletal Position=true
    Initial Color:LinearColor=$Particles.Initial.Color
    Initial RibbonWidth:NiagaraFloat=$Particles.Initial.RibbonWidth
   )
   &ScaleRibbonWidth(Ribbon Width Scale=&FloatFromCurve(
     FloatCurve=DataInterface() {propertyValues {Curve {keys [
         {arriveTangent=4.9346, interpMode=RCIM_Cubic, leaveTangent=4.9346, tangentMode=RCTM_User, time=0, value=0}
         <at 0.15 1 RCIM_Cubic>
         {arriveTangent=-2.0105, interpMode=RCIM_Cubic, leaveTangent=-2.0105, tangentMode=RCTM_User, time=1, value=0}
        ]}}}
     Scale Curve=0.5
    ))
   &KillParticles(Kill Particles=&SetBoolByFloatComparison(
     A=&MakeCustomFloatFromBool(Bool=$User.bIsFrontEndPreview, True Float=1, False Float=$System.FN_Integrator.NormalizedValue)
     B=0
     Comparison Type=::NewEnumerator4
    ))
  ]
  rendererClasses [&NiagaraRibbonRendererProperties]
  renderers [{rendererClass=&NiagaraRibbonRendererProperties}]
 }
 GoopRibbonsBurst {
  bEnabled=false
  emitterUpdateScript [
   &EmitterState(Life Cycle Mode=::NewEnumerator1, Loop Duration=1, UseLoopDelay=true)
   &FN_SpawnRibbon_v2(Number of Ribbons=4, SpawnRate=40)
  ]
  particleSpawnScript [
   /SetVariables(Particles.MaterialRandom:NiagaraFloat=&RandomRangeFloat(Minimum=0, Maximum=1)) {isSetParameters=true, module=SetVariables_#1}
   &FN_InitializeRibbon_v3(
    Enable Skeletal Position=true
    Flip Facing=true
    Lifetime=0.4
    Ribbon Width=20
    Sampled Position Coordinate Space:ENiagaraCoordinateSpace=::Local
    Sampled Rotation Coordinate Space:ENiagaraCoordinateSpace=::Local
    Sampled Velocity Coordinate Space:ENiagaraCoordinateSpace=::Local
   ) {module=FN_InitializeRibbon_v3001}
   /SetVariables(Particles.RandomizeLoopedEmitterAge:NiagaraFloat=&NormalizeFloat(
     Value=&Add_Float(A=&RandomRangeFloat(Minimum=0, Maximum=1), B=$Emitter.NormalizedLoopAge)
     Min=0
     Max=2
    )) {isSetParameters=true, module=SetVariables_#2}
   &AddVelocity(
    Coordinate Space=::Simulation
    Velocity=&RotateVectorByQuaternion(
     Vector To Rotate=&Multiply_Vector(
      A=<100 100 100>
      B=&SelectVectorFromArray(
       Direct Array Index=$Particles.SpawnGroup
       Vector Selection Array=DataInterface() {propertyValues {FloatData [<0 -1 0>, <0 -1 0>, <0 1 0>, <0 1 0>]}}
      )
     )
     Quaternion=$Particles.SampledRotation
    )
   ) {module=AddVelocity001}
   &AddVelocityFromPoint(Module.Use Falloff Distance=false, Velocity Strength=&RandomRangeFloat(Minimum=10, Maximum=50))
  ]
  particleUpdateScript [
   &CurlNoiseForce(
    Noise Frequency=1
    Noise Strength=&Multiply_Float(
     A=500
     B=&FloatFromCurve(FloatCurve=DataInterface() {propertyValues {Curve {keys [{tangentMode=RCTM_User, time=0, value=1}, {tangentMode=RCTM_User, time=1, value=1}]}}})
    )
    Pan Noise=true
   )
   &DynamicMaterialParameters(
    Index 0 Param 1=1
    Index 0 Param 2=$Particles.MaterialRandom
    Index 0 Param 3=5
    Index 0 Param 4=$User.EmissiveIntensity
    Module.Param0WriteEnabled=true
    Module.Param1WriteEnabled=true
    Module.Param2WriteEnabled=true
    Module.Param3WriteEnabled=true
   )
   &SolveForcesAndVelocity(Clamp Velocity=false, Limit Acceleration=false)
   &FN_UpdateRibbon(
    Enable Skeletal Position=true
    Initial Color:LinearColor=$Particles.Initial.Color
    Initial RibbonWidth:NiagaraFloat=$Particles.Initial.RibbonWidth
   )
   &ScaleRibbonWidth(Ribbon Width Scale=&FloatFromCurve(FloatCurve=DataInterface() {propertyValues {Curve {keys [
         <at 0 0.2 RCIM_Cubic>
         {arriveTangent=-0.2, interpMode=RCIM_Cubic, leaveTangent=-0.2, time=0.3, value=1}
         {arriveTangent=-2.0105, interpMode=RCIM_Cubic, leaveTangent=-2.0105, tangentMode=RCTM_User, time=1, value=0}
        ]}}}))
  ]
  rendererClasses [&NiagaraRibbonRendererProperties]
  renderers [{rendererClass=&NiagaraRibbonRendererProperties}]
 }
 SideWings {
  emitterSpawnScript [&FN_RibbonIDOffsetUtility()]
  emitterUpdateScript [
   &EmitterState(Delay First Loop Only=true, Life Cycle Mode=::NewEnumerator1, Loop Delay=1, Loop Duration=1, UseLoopDelay=true)
   &FN_SpawnAnimTrail(Number of Anim Trails=2)
  ]
  particleSpawnScript [
   &FN_InitializeAnimTrailBoneIndecies(Second Bone or Socket Index=6)
   &FN_InitializeAnimTrailBoneIndecies(AnimTrail Selection=1, First Bone or Socket Index=3, Second Bone or Socket Index=7) {module=FN_InitializeAnimTrailBoneIndecies001}
   &FN_InitializeAnimTrail(
    AnimTrailWidth=&FN_FrontEndFloatModifier(Gameplay=1.5, Front End=2)
    Color=<rgba 1 0.724 0.724 1>
    Lifetime=0.2
    SkeletalMesh=$User.Skeletal Mesh Drops
   )
   &DynamicMaterialParameters(
    Index 0 Param 1=$Particles.MaterialRandom
    Index 0 Param 2=$Particles.MaterialRandom
    Module.Param0WriteEnabled=true
    Module.Param1WriteEnabled=true
    Module.Param2WriteEnabled=true
    Module.Param3WriteEnabled=true
   )
   &AddVelocity(Velocity=&FN_FrontEndVector3Modifier(Gameplay=<0 2500 0>, Front End=<0 -2500 0>)) {module=AddVelocity001}
  ]
  particleUpdateScript [
   &SolveForcesAndVelocity(Clamp Velocity=false, Limit Acceleration=false)
   &FN_UpdateAnimTrail(SkeletalMesh=$User.SkeletalMesh) {module=FN_UpdateAnimTrail001}
   /SetVariables(Particles.RibbonU0Override=$Particles.NormalizedAge) {isSetParameters=true}
   &ScaleColor(
    Scale Alpha=&FN_FrontEndFloatModifier(Gameplay=&Lerp_Float(A=0, B=0.4, Alpha=$System.FN_Integrator.NormalizedValue), Front End=0.2)
    ScaleA=true
    ScaleRGB=true
   ) {module=ScaleColor001}
  ]
  rendererClasses [&NiagaraRibbonRendererProperties]
  renderers [{rendererClass=&NiagaraRibbonRendererProperties}]
 }
}
summary {
 systemName=NS_Glider_Clash_Idle
 emitters [
  {bEnabled=false, emitterName=GoopRibbonsBurst, rendererClasses [&NiagaraRibbonRendererProperties]}
  {emitterName=SideWings, rendererClasses [&NiagaraRibbonRendererProperties]}
  {bEnabled=false, emitterName=Drops, rendererClasses [&NiagaraSpriteRendererProperties]}
  {bEnabled=false, emitterName=GoopRibbons, rendererClasses [&NiagaraRibbonRendererProperties]}
 ]
 userVariables [
  User.bIsFrontEnd {type=NiagaraBool}
  User.bIsFrontEndPreview {type=NiagaraBool}
  User.bisFullyDeployed {type=NiagaraBool}
  User.ForwardDot {type=NiagaraFloat, defaultValue=0, description=""}
  User.RibbonIDOffset {type=NiagaraInt32}
  User.RightDot {type=NiagaraFloat, defaultValue=0, description=""}
  User.Skeletal Mesh Drops {
   type=NiagaraDataInterfaceSkeletalMesh
   defaultValue {
    struct=NiagaraExt_VariableValue_DataInterface
    value {dataInterface=&self:NiagaraDataInterfaceSkeletalMesh_2}
   }
  }
  User.Skeletal Mesh Tendrils {
   type=NiagaraDataInterfaceSkeletalMesh
   defaultValue {
    struct=NiagaraExt_VariableValue_DataInterface
    value {dataInterface=&self:NiagaraDataInterfaceSkeletalMesh_1}
   }
  }
  User.SkeletalMesh {
   type=NiagaraDataInterfaceSkeletalMesh
   defaultValue {
    struct=NiagaraExt_VariableValue_DataInterface
    value {dataInterface=&self:NiagaraDataInterfaceSkeletalMesh_0}
   }
  }
 ]
}
